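# Physics asset for the shorts mesh of the M_MED_Ohana male player cosmetic body, setting up secondary motion on the shorts bones.
# Each side has a larger capsule on its dyn_shorts1 bone, offset in mirror image, and a tiny capsule on the thigh bone.
# A constraint ties each shorts bone to its thigh with a very narrow swing limit and a zero twist limit, keeping movement subtle.
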PhysicsAsset /BRCosmetics/Characters/Player/Male/Medium/Bodies/M_MED_Ohana/Meshes/M_MED_Ohana_Shorts_Physics
bodies {
 MSTR_thigh_l [{shapeType=Capsule, length=0.505, radius=0.505, shapeName=MSTR_thigh_l_capsule, rotation=<rot 0 -90 0>}]
 MSTR_thigh_r [{shapeType=Capsule, length=0.505, radius=0.505, shapeName=MSTR_thigh_r_capsule, rotation=<rot 0 -90 0>}]
 dyn_shorts1_l [{shapeType=Capsule, length=5, radius=5, shapeName=dyn_shorts1_l_capsule, center=<-5 0 0>, rotation=<rot 90 0 0>}]
 dyn_shorts1_r [{shapeType=Capsule, length=5, radius=5, shapeName=dyn_shorts1_r_capsule, center=<5 0 0>, rotation=<rot 90 0 0>}]
}
constraints [
 {bone1Name=dyn_shorts1_l, bone2Name=MSTR_thigh_l, swing1LimitDegrees=3, swing2LimitDegrees=3, twistLimitDegrees=0}
 {bone1Name=dyn_shorts1_r, bone2Name=MSTR_thigh_r, swing1LimitDegrees=3, swing2LimitDegrees=3, twistLimitDegrees=0}
]
