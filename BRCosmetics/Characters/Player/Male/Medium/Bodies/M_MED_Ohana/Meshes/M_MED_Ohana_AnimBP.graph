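# Post-process style cosmetic AnimBP for the male medium Ohana body, layering deformation and cloth motion over the pose copied from the parent mesh.
# Shirt and back Control Rigs feed two RigidBody sims with overridden gravity, the shorts sim active only while crouching or sliding.
# Bone constraints and additive offsets correct shorts, towel, arm and backpack.
# EventGraph derives their alphas from knee bend, movement state, pistol-aim curve and tagged-backpack emote curve.
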AnimBlueprint /BRCosmetics/Characters/Player/Male/Medium/Bodies/M_MED_Ohana/Meshes/M_MED_Ohana_AnimBP
target_skeleton=&/M_MED_Ohana_Skeleton
graphs [
 AnimGraph {nodes [
   n1 = Control Rig(Source=Copy Pose From Mesh() {bCopyCurves=true, bUseAttachedParent=true}) {
    alphaBoolBlend {blendOption=Linear}
    controlRigAssetReference {blueprintRigClass=&/M_MED_DeformRig_OhanaShirt_ControlRig_new.M_MED_DeformRig_OhanaShirt_ControlRig_new_C}
   }
   n2 = Control Rig(Source=@n1) {
    alphaBoolBlend {blendOption=Linear}
    controlRigAssetReference {blueprintRigClass=&/M_MED_Ohana__Back_CtrlRig.M_MED_Ohana__Back_CtrlRig_C}
   }
   n3 = "Transform (Modify) Bone - Bone: deform_upperarm_fr_r"(
    ComponentPose=Local To Component(LocalPose=@n2)
    Alpha=Get pistolPoseFix() -> (pistolPoseFix:double)
    Translation="50,0,0"
    Rotation="0,0,0"
    Scale="1,1,1"
   ) {
    translationMode=BMM_Additive
    translationSpace=BCS_BoneSpace
    alphaBoolBlend {blendOption=Linear}
    boneToModify {boneName=deform_upperarm_fr_r}
    translation {x=50}
   }
   n4 = Get GravityOverride()
   n5 = RigidBody!(ComponentPose=@n3, Alpha, OverrideWorldGravity:Vector=@n4, ExternalForce) -> (Pose) {
    bClampLinearTranslationLimitToRefPose=true
    bOverrideWorldGravity=true
    overridePhysicsAsset=&/M_MED_Ohana_Physics1
    simulationSpace=BaseBoneSpace
    alphaBoolBlend {blendOption=Linear}
    baseBoneRef {boneName=root}
    componentAppliedLinearAccClamp=<3000 3000 3000>
    componentLinearAccScale=<6 0.3 0.3>
    componentLinearVelScale=<6 0.3 0.3>
    overrideWorldGravity {z=-980}
    simSpaceSettings {
     maxAngularAcceleration=1000
     maxAngularVelocity=3
     maxLinearAcceleration=700
     maxLinearVelocity=500
     velocityScaleZ=0
     worldAlpha=0.25
     externalLinearDragV=<0.2 0.2>
    }
   }
   n6 = RigidBody!(ComponentPose=@n5, bAlphaBoolEnabled:bool=Get isCrouchingSliding() -> (isCrouchingSliding:bool), OverrideWorldGravity=@n4, ExternalForce) -> (Pose) {
    alphaInputType=Bool
    bClampLinearTranslationLimitToRefPose=true
    bOverrideWorldGravity=true
    overridePhysicsAsset=&/M_MED_Ohana_Shorts_Physics
    simulationSpace=BaseBoneSpace
    alphaBoolBlend {blendOption=Linear}
    baseBoneRef {boneName=root}
    componentAppliedLinearAccClamp=<3000 3000 3000>
    componentLinearAccScale=<6 0.3 0.3>
    componentLinearVelScale=<6 0.3 0.3>
    overrideWorldGravity {z=-980}
    simSpaceSettings {
     maxAngularAcceleration=1000
     maxAngularVelocity=3
     maxLinearAcceleration=700
     maxLinearVelocity=500
     velocityScaleZ=0
     externalLinearDragV=<0.2 0.2>
    }
   }
   n7 = "Constraint - dyn_shorts2_r (RR)"!(ComponentPose=@n6, Alpha=1, ConstraintWeights_0=1, ConstraintWeights_1:float=0) -> (Pose) {
    alphaBoolBlend {blendOption=Linear}
    boneToModify {boneName=dyn_shorts2_r}
    constraintSetup [{transformType=Rotation, targetBone {boneName=thigh_r}}, {transformType=Rotation, targetBone {boneName=MSTR_thigh_r}}]
    constraintWeights [1,1]
   }
   n8 = "Transform (Modify) Bone - Bone: dyn_shorts2_l"(ComponentPose=@n7, Alpha=Get KneeLf() -> (KneeLf:double), Translation="8,4,0", Rotation="0,60,0", Scale="1,1,1") {
    rotationMode=BMM_Additive
    rotationSpace=BCS_BoneSpace
    scaleMode=BMM_Additive
    translationMode=BMM_Additive
    translationSpace=BCS_BoneSpace
    alphaBoolBlend {blendOption=Linear}
    alphaScaleBiasClamp {bInterpResult=true, interpSpeedIncreasing=500}
    boneToModify {boneName=dyn_shorts2_l}
    rotation {yaw=60}
    translation=<8 4>
   }
   n9 = "Transform (Modify) Bone - Bone: dyn_shorts2_r"(ComponentPose=@n8, Alpha=Get KneeRt() -> (KneeRt:double), Translation="0,5,2", Rotation="0,90,0", Scale="1,1,1") {
    rotationMode=BMM_Additive
    rotationSpace=BCS_BoneSpace
    scaleMode=BMM_Additive
    scaleSpace=BCS_BoneSpace
    translationMode=BMM_Additive
    alphaBoolBlend {blendOption=Linear}
    alphaScaleBiasClamp {bInterpResult=true, interpSpeedIncreasing=500}
    boneToModify {boneName=dyn_shorts2_r}
    rotation {yaw=90}
    translation {y=5, z=2}
   }
   n10 = "Transform (Modify) Bone - Bone: dyn_shorts2_r"(ComponentPose=@n9, Alpha=Get Is Jumping() -> (Is Jumping:double), Translation="0,-5,2", Rotation="0,0,0", Scale="1,1,1") {
    rotationMode=BMM_Additive
    rotationSpace=BCS_BoneSpace
    scaleMode=BMM_Additive
    scaleSpace=BCS_BoneSpace
    translationMode=BMM_Additive
    alphaBoolBlend {blendOption=Linear}
    alphaScaleBiasClamp {bInterpResult=true, interpSpeedDecreasing=3, interpSpeedIncreasing=500}
    boneToModify {boneName=dyn_shorts2_r}
    translation {y=-5, z=2}
   }
   n11 = "Transform (Modify) Bone - Bone: dyn_shorts2_r"!(ComponentPose=@n10, bAlphaBoolEnabled:bool=Get bIsSliding(), Translation="0,2,2", Rotation="0,0,0", Scale="1,1,1") -> (Pose) {
    alphaInputType=Bool
    rotationMode=BMM_Additive
    rotationSpace=BCS_BoneSpace
    scaleMode=BMM_Additive
    scaleSpace=BCS_BoneSpace
    translationMode=BMM_Additive
    alphaBoolBlend {blendOption=Linear}
    alphaScaleBiasClamp {bInterpResult=true, interpSpeedIncreasing=500}
    boneToModify {boneName=dyn_shorts2_r}
    translation {y=2, z=2}
   }
   n12 = "Transform (Modify) Bone - Bone: pelvis_towel"!(ComponentPose=@n11, bAlphaBoolEnabled:bool=True, Translation="0,2,0", Rotation="0,0,0", Scale="0.5,0.5,0.5") -> (Pose) {
    alphaInputType=Bool
    rotationMode=BMM_Additive
    rotationSpace=BCS_BoneSpace
    scaleMode=BMM_Additive
    scaleSpace=BCS_BoneSpace
    translationMode=BMM_Additive
    alphaBoolBlend {blendOption=Linear}
    alphaScaleBiasClamp {bInterpResult=true, interpSpeedIncreasing=500}
    boneToModify {boneName=pelvis_towel}
    scale=<0.5 0.5 0.5>
    translation {y=2}
   }
   n13 = "Transform (Modify) Bone - Bone: attach_backpack"(ComponentPose=@n12, Alpha=Get OhanaEmote() -> (OhanaEmote:double), Translation="0,-1,1", Rotation="-5,-5,5", Scale="1,1,1") {
    rotationMode=BMM_Additive
    rotationSpace=BCS_BoneSpace
    translationMode=BMM_Additive
    translationSpace=BCS_BoneSpace
    alphaBoolBlend {blendOption=Linear}
    boneToModify {boneName=attach_backpack}
    rotation=<rot -5 -5 5>
    translation {y=-1, z=1}
   }
   Output Pose(Result=Component To Local(ComponentPose=@n13)) {layerGroup=None}
  ]}
 EventGraph {nodes [
   Event BlueprintUpdateAnimation()
   > n1 = Sequence!(execute) -> (then_0, then_1, then_2:Exec)
   > n2 = Cast To FortPlayerAnimInstance(Object=GetAnimInstance(self=GetOwnerSkeletalMeshComponent()))
   > n3 = Set isPistol(isPistol="Equal (Enum)"(A:EFortWeaponCoreAnimation Enum=Get WeaponCoreAnim(self=@n2), B:EFortWeaponCoreAnimation Enum=Pistol))
   n4 = Get bIsSkyDiving()
   > n5 = Branch(Condition=AND Boolean(A=@n3, B=NOT Boolean(A=@n4)))
   > Set pistolPoseFix(pistolPoseFix:double=GetCurveValue(CurveName=ADS_Pistol_Up)) -> (Output_Get:double)
   Set pistolPoseFix(execute=@n5.else, pistolPoseFix=0)
   n6 = Gravity Override State Function DEPRECATED(
    execute=@n1.then_1
    CharacterPartAnimInstance=Self-Reference()
    GameTimeInSeconds=GetGameTimeInSeconds()
    GravityOverrideParams=Get Gravity Override Params()
   )
   > Set GravityOverride(GravityOverride=@n6)
   n7 = BreakTransform(InTransform=GetDeltaTransformFromRefPose(self=GetOwnerSkeletalMeshComponent(), BoneName=calf_l, BaseName=thigh_l))
   > Set KneeLf(KneeLf:double=MapRangeClamped(Value=BreakRotator(InRot=@n7.Rotation).Yaw, InRangeA=50, InRangeB=150)) -> (Output_Get:double)
   n8 = BreakTransform(InTransform=GetDeltaTransformFromRefPose(self=GetOwnerSkeletalMeshComponent(), BoneName=calf_r, BaseName=thigh_r))
   > Set KneeRt(KneeRt:double=MapRangeClamped(Value=BreakRotator(InRot=@n8.Rotation).Yaw, InRangeA=50, InRangeB=150)) -> (Output_Get:double)
   > Set Is Jumping(Is Jumping:double="To Float (Boolean)"(InBool=Get bIsJumping())) -> (Output_Get:double)
   > Set isCrouchingSliding(isCrouchingSliding:bool=OR Boolean(A=Get bIsSliding(), B=Get bIsCrouching())) -> (Output_Get:bool)
   n9 = Branch(execute=@n1.then_2, Condition=Get bIsPlayingEmote())
   > n10 = Branch(Condition=BackpackItemDefinitionHasAnyMetaTag!(Tags:Array of Gameplay Tag Structures=Make Array("[0]"=(TagName="Cosmetics.MetaTags.Ohana")), bRequiresEquipped:bool=false) -> (ReturnValue:bool))
   > Set OhanaEmote(OhanaEmote:double=GetCurveValue(self=Get AnimInst() -> (AnimInst:Anim Instance Object Reference), CurveName=OhanaPackAdj)) -> (Output_Get:double)
   Set OhanaEmote(execute=@n10.else, OhanaEmote=0)
   Set OhanaEmote(execute=@n9.else, OhanaEmote=0)
   Event BlueprintInitializeAnimation()
   n11 = GetOwnerSkeletalMeshComponent()
   > Is Valid(InputObject=@n11)
   > Set Skeletal_Mesh(Skeletal_Mesh:Skeletal Mesh Component Object Reference=@n11) -> (Output_Get:Skeletal Mesh Component Object Reference)
   > n12 = Cast To FortPlayerPawn(Object=GetOwningActor())
   > Is Valid(InputObject=@n12)
   > n13 = Set Pawn(Pawn:Fort Player Pawn Object Reference=@n12) -> (Output_Get:Fort Player Pawn Object Reference)
   n14 = GetAnimInstance(self=Get Mesh(self=@n13))
   > Is Valid(InputObject=@n14)
   > Set AnimInst(AnimInst:Anim Instance Object Reference=@n14) -> (Output_Get:Anim Instance Object Reference)
  ]}
]
variables [
 AnimInst {type=AnimInstance, default=null}
 Gravity Override Params {type=GravityOverrideParamsStruct, default {skydive_FloatingDown {gravityOverride=<0 0 -980>}}}
 GravityOverride {type=Vector}
 Is Jumping {type=number}
 isCrouchingSliding {type=boolean}
 isPistol {type=boolean}
 KneeLf {type=number}
 KneeRt {type=number}
 OhanaEmote {type=number}
 Pawn {type=FortPlayerPawn}
 pistolPoseFix {type=number}
 Skeletal_Mesh {type=SkeletalMeshComponent}
 SkeletalMesh {type=SkeletalMeshComponent}
]
